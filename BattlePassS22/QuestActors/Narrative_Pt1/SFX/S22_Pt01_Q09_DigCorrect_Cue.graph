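# Sound cue for the correct-dig feedback in Season 22 battle pass narrative Part 1, Quest 09.
# A Random node picks one of three DigCorrect wave variants at equal weight.
# It plays through the shared Season 22 quest attenuation and the Fort_Quests sound class.
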SoundCue /BattlePassS22/QuestActors/Narrative_Pt1/SFX/S22_Pt01_Q09_DigCorrect_Cue
root=Random_0
nodes [
 {id=Random_0, children [WavePlayer_0,WavePlayer_1,WavePlayer_2], properties {weights [1,1,1]}}
 {id=WavePlayer_0, properties {soundWaveAssetPtr=&/S22_Pt01_Q09_DigCorrect_01}}
 {id=WavePlayer_1, properties {soundWaveAssetPtr=&/S22_Pt01_Q09_DigCorrect_02}}
 {id=WavePlayer_2, properties {soundWaveAssetPtr=&/S22_Pt01_Q09_DigCorrect_03}}
]
properties {
 attenuationSettings=&/S22_Quests_Atten
 maxDistance=1300
 soundClassObject=&Fort_Quests
 volumeMultiplier=1
}
