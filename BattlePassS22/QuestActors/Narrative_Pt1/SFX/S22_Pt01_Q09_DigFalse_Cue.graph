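# Sound cue for a Season 22 narrative quest (Part 1, Q09), likely the one-shot for a false or unsuccessful dig.
# A Random node picks one of five DigFalse wave variants with equal weight on each play.
# Playback routes through the quests sound class and uses the shared Season 22 quest attenuation asset for spatial falloff.
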SoundCue /BattlePassS22/QuestActors/Narrative_Pt1/SFX/S22_Pt01_Q09_DigFalse_Cue
root=Random_1
nodes [
 {id=Random_1, children [WavePlayer_3,WavePlayer_4,WavePlayer_5,WavePlayer_6,WavePlayer_7], properties {weights [1,1,1,1,1]}}
 {id=WavePlayer_3, properties {soundWaveAssetPtr=&/S22_Pt01_Q09_DigFalse_01}}
 {id=WavePlayer_4, properties {soundWaveAssetPtr=&/S22_Pt01_Q09_DigFalse_02}}
 {id=WavePlayer_5, properties {soundWaveAssetPtr=&/S22_Pt01_Q09_DigFalse_03}}
 {id=WavePlayer_6, properties {soundWaveAssetPtr=&/S22_Pt01_Q09_DigFalse_04}}
 {id=WavePlayer_7, properties {soundWaveAssetPtr=&/S22_Pt01_Q09_DigFalse_05}}
]
properties {
 attenuationSettings=&/S22_Quests_Atten
 maxDistance=1300
 soundClassObject=&Fort_Quests
 volumeMultiplier=1
}
