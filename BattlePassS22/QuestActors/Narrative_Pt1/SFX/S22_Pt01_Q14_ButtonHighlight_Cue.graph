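# Sound cue that plays the button-highlight sound for quest 14 in part one of the Season 22 Battle Pass narrative quests.
# A single wave player feeds the S22_Pt01_Q14_ButtonHighlight wave straight to the output, with no randomization or modulation.
# It routes through the Fort_Quests sound class and sets a very large max distance.
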SoundCue /BattlePassS22/QuestActors/Narrative_Pt1/SFX/S22_Pt01_Q14_ButtonHighlight_Cue
root=WavePlayer_1
nodes [{id=WavePlayer_1, properties {soundWaveAssetPtr=&/S22_Pt01_Q14_ButtonHighlight}}]
properties {maxDistance=2097152, soundClassObject=&Fort_Quests, volumeMultiplier=1}
