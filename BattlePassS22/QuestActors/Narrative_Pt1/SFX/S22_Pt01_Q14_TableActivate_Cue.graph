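# Sound cue for a table activation effect, named for Season 22 Battle Pass narrative Part 1 quest 14.
# A single wave player plays the TableActivate sound wave directly, with no randomization, mixing or modulation nodes.
# The cue routes to the Fort_Quests sound class with boosted volume and a very large max distance.
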SoundCue /BattlePassS22/QuestActors/Narrative_Pt1/SFX/S22_Pt01_Q14_TableActivate_Cue
root=WavePlayer_0
nodes [{id=WavePlayer_0, properties {soundWaveAssetPtr=&/S22_Pt01_Q14_TableActivate}}]
properties {maxDistance=2097152, soundClassObject=&Fort_Quests, volumeMultiplier=1.5}
