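# Looping sound cue for a table in Battle Pass Season 22 Narrative Part 1 quest 14, played under the quest sound class.
# A looping wave player feeds an Enveloper whose volume curve fades the loop in from silence over the first second.
# The cue uses a dedicated table attenuation asset and sits slightly below full volume.
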SoundCue /BattlePassS22/QuestActors/Narrative_Pt1/SFX/S22_Pt01_Q14_Table_LP_Cue
root=Enveloper_0
nodes [
 {
  id=Enveloper_0
  children [WavePlayer_0]
  properties {pitchCurve {editorCurveData {keys [{time=0}]}}, volumeCurve {editorCurveData {keys [<at 0 0>,<at 1 1>]}}}
 }
 {id=WavePlayer_0, properties {bLooping=true, soundWaveAssetPtr=&/S22_Pt01_Q14_Table_LP}}
]
properties {
 attenuationSettings=&/S22_Pt01_Q14_Table_Atten
 maxDistance=1700
 soundClassObject=&Fort_Quests
 volumeMultiplier=0.85
}
